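# Control Rig for the female CherryKnot Strings body that applies Fortnite deform-rig correctives and adjusts costume dyn bones.
# After the deform rig, functions offset belt, front loincloth and side skirt bones using thigh pose readers, deform-rig alpha attributes and isCrouch.
# Loincloth roots half-follow twist-filtered thigh helper bones via parent constraints.
# An upper-arm pose reader shifts left bow bones; construction adds FK controls, and DeltaRotation is defined but unused.
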ControlRigBlueprint /BRCosmetics/Characters/Player/Female/Medium/Bodies/F_MED_CherryKnot_Strings/Meshes/F_MED_CherryKnot_Strings_CtrlRig
graphs [
 New Function_ContainedGraph {nodes [
   n1 = RigVMFunctionEntryNode()
   RigVMFunction_MathBoolToInteger = &RigVMFunction_MathBoolToInteger(Value=@n1.onlyTwist)
   RigVMFunction_MathQuaternionSwingTwist = &RigVMFunction_MathQuaternionSwingTwist(
    TwistAxis=@n1.twistAxis
    Input=&RigUnit_GetRelativeTransformForItem(bChildInitial=False, bParentInitial=False, Child=@n1.child, Parent=@n1.parent).RelativeTransform.Rotation
   )
   n2 = RigVMDispatchNode(
    Values=((),())
    Index=@RigVMFunction_MathBoolToInteger.Result
    Values.0=@RigVMFunction_MathQuaternionSwingTwist.Swing
    Values.1=@RigVMFunction_MathQuaternionSwingTwist.Twist
   ) {template="DISPATCH_RigVMDispatch_SelectInt32(in Index,in Values,out Result)"}
   RigVMFunction_MathQuaternionToEuler = &RigVMFunction_MathQuaternionToEuler(Value=@n2.Result, RotationOrder=@n1.Rotation Order)
   > DISPATCH_RigVMDispatch_Print = RigVMDispatchNode(ScreenColor=(R=1,G=1,B=1,A=1), ScreenDuration=1.5, Enabled=@n1.debugRotation, Value=@RigVMFunction_MathQuaternionToEuler.Result) {template="DISPATCH_RigVMDispatch_Print(in Prefix,in Value,in Enabled,in ScreenDuration,in ScreenColor)"}
   Cast = RigVMDispatchNode(Value=@n1.triggerAxis) {template="DISPATCH_RigVMDispatch_CastEnumToInt(in Value,out Result)"}
   DISPATCH_RigVMDispatch_SelectInt32 = RigVMDispatchNode(
    Values=(0,0,0)
    Index=@Cast.Result
    Values.0=@RigVMFunction_MathQuaternionToEuler.Result.X
    Values.1=@RigVMFunction_MathQuaternionToEuler.Result.Y
    Values.2=@RigVMFunction_MathQuaternionToEuler.Result.Z
   ) {template="DISPATCH_RigVMDispatch_SelectInt32(in Index,in Values,out Result)"}
   Remap = &RigVMFunction_MathDoubleRemap(
    bClamp=False
    Value=@DISPATCH_RigVMDispatch_SelectInt32.Result
    TargetMinimum=@n1.outputMinMax.X
    TargetMaximum=@n1.outputMinMax.Y
    SourceMinimum=@n1.rotationMinMax.X
    SourceMaximum=@n1.rotationMinMax.Y
   )
   Absolute = &RigVMFunction_MathDoubleAbs(Value=@Remap.Result)
   Multiply = &RigVMFunction_MathDoubleMul(
    A=&RigVMFunction_MathDoubleClamp(Value=RigVMDispatchNode(True=@Absolute.Result, Condition=@n1.absoluteValue, False=@Remap.Result) {template="DISPATCH_RigVMDispatch_If(in Condition,in True,in False,out Result)"}.Result).Result
    B=@n1.premult
   )
   > RigVMFunctionReturnNode(outputQuat=@n2.Result, outputWeight=@Multiply.Result)
  ]}
 New Function_ContainedGraph {nodes [
   RigVMFunctionEntryNode()
   Get_Animation_Attribute_1 = RigVMDispatchNode(BoneName=None, Default=0, Name=deformRig_r_thigh_up_alpha) {template="DISPATCH_RigDispatch_GetAnimAttribute(in Name,in BoneName,in Default,out Value,out Found)"}
   > &RigUnit_ModifyTransforms(
    ItemToModify=((Item=(Type=Bone,Name="dyn_belt_hipAcc_r_MSTR"),Transform=(Rotation=(X=-0.1153,Y=-0.0075,Z=0.0093,W=0.9933),Translation=(X=-1.0318,Y=-2.542,Z=-1.868),Scale3D=(X=1,Y=1,Z=1))))
    WeightMinimum=0
    Weight=@Get_Animation_Attribute_1.Value
   )
   > n1 = &RigUnit_SphericalPoseReader(
    ActiveRegionScaleFactors=(PositiveWidth=3,NegativeWidth=1,PositiveHeight=1,NegativeHeight=1)
    ActiveRegionSize=0.1
    Debug=(bDrawDebug=false,bDraw2D=False,bDrawLocalAxes=False,DebugScale=25,DebugSegments=20,DebugThickness=0.25)
    DriverAxis=(X=1,Y=0,Z=0)
    DriverItem=(Type=Bone,Name="thigh_r")
    FalloffRegionScaleFactors=(PositiveWidth=1,NegativeWidth=1,PositiveHeight=1,NegativeHeight=1)
    FalloffSize=0.5
    FlipHeightScaling=true
    FlipWidthScaling=true
    OptionalParentItem=(Type=Bone,Name="None")
    RotationOffset=(X=0,Y=0,Z=0)
   )
   > &RigUnit_ModifyTransforms(
    ItemToModify=((Item=(Type=Bone,Name="dyn_belt_hipAcc_r_MSTR"),Transform=(Rotation=(X=0.0311,Y=-0.63,Z=0.0389,W=-0.775),Translation=(X=-4.6496,Y=0.1674,Z=-3.6979),Scale3D=(X=1,Y=1,Z=1))))
    WeightMinimum=0
    Weight=@n1.OutputParam
   )
   > RigVMFunctionReturnNode()
  ]}
 New Function_ContainedGraph {nodes [
   RigVMFunctionEntryNode()
   RigUnit_GetTransform = &RigUnit_GetTransform(Item=(Type=Bone,Name="thigh_l"), Space=LocalSpace, bInitial=true)
   RigVMFunction_MathQuaternionSwingTwist = &RigVMFunction_MathQuaternionSwingTwist(
    TwistAxis=(X=1,Y=0,Z=0)
    Input=&RigVMFunction_MathTransformMakeRelative(
     Global=@RigUnit_GetTransform.Transform
     Parent=&RigUnit_GetTransform(Item=(Type=Bone,Name="thigh_l"), Space=LocalSpace, bInitial=False).Transform
    ).Local.Rotation
   )
   > RigUnit_SetRotation = &RigUnit_SetRotation(Item=(Type=Bone,Name="cr_thigh_l_no_twist"), Weight=1, bInitial=False, Value=@RigVMFunction_MathQuaternionSwingTwist.Twist)
   n1 = &RigUnit_GetTransform(Item=(Type=Bone,Name="thigh_r"), Space=LocalSpace, bInitial=true)
   n2 = &RigUnit_GetTransform(Item=(Type=Bone,Name="thigh_r"), Space=LocalSpace, bInitial=False)
   > &RigUnit_SetRotation(
    Item=(Type=Bone,Name="cr_thigh_r_no_twist")
    Weight=1
    bInitial=False
    Value=&RigVMFunction_MathQuaternionSwingTwist(
     TwistAxis=(X=-1,Y=0,Z=0)
     Input=&RigVMFunction_MathTransformMakeRelative(Global=@n1.Transform, Parent=@n2.Transform).Local.Rotation
    ).Twist
   )
   > &RigUnit_ParentConstraint(
    AdvancedSettings=(InterpolationType=Average,RotationOrderForFilter=XZY)
    Child=(Type=Bone,Name="dyn_loincloth_front_l_root")
    Parents=((Item=(Type=Bone,Name="cr_thigh_l_no_twist"),Weight=1))
    Weight=0.5
    bMaintainOffset=true
   )
   > &RigUnit_ParentConstraint(
    AdvancedSettings=(InterpolationType=Average,RotationOrderForFilter=XZY)
    Child=(Type=Bone,Name="dyn_loincloth_front_r_root")
    Parents=((Item=(Type=Bone,Name="cr_thigh_r_no_twist"),Weight=1))
    Weight=0.5
   )
   > &RigUnit_ParentConstraint(
    AdvancedSettings=(InterpolationType=Average,RotationOrderForFilter=XZY)
    Child=(Type=Bone,Name="dyn_loincloth_front_root")
    Parents=((Item=(Type=Bone,Name="cr_thigh_l_no_twist"),Weight=0.5),(Item=(Type=Bone,Name="cr_thigh_r_no_twist"),Weight=0.5))
    Weight=0.5
    bMaintainOffset=true
   )
   Get_Animation_Attribute_2 = RigVMDispatchNode(BoneName=None, Default=0, Name=deformRig_r_thigh_up_alpha) {template="DISPATCH_RigDispatch_GetAnimAttribute(in Name,in BoneName,in Default,out Value,out Found)"}
   > &RigUnit_ModifyTransforms(
    ItemToModify=((Item=(Type=Bone,Name="dyn_loincloth_front_r_root"),Transform=(Rotation=(X=-0.4241,Y=0.0429,Z=0.0728,W=0.9016),Translation=(X=-5,Y=-6,Z=0),Scale3D=(X=1,Y=1,Z=1))))
    WeightMinimum=0
    Weight=@Get_Animation_Attribute_2.Value
   )
   Get_Animation_Attribute_1 = RigVMDispatchNode(BoneName=None, Default=0, Name=deformRig_l_thigh_up_alpha) {template="DISPATCH_RigDispatch_GetAnimAttribute(in Name,in BoneName,in Default,out Value,out Found)"}
   > &RigUnit_ModifyTransforms(
    ItemToModify=((Item=(Type=Bone,Name="dyn_loincloth_front_root"),Transform=(Rotation=(X=0,Y=0,Z=0,W=1),Translation=(X=0,Y=-1.8,Z=0),Scale3D=(X=1,Y=1,Z=1))))
    WeightMinimum=0
    Weight=@Get_Animation_Attribute_1.Value
   )
   Get_Animation_Attribute_3 = RigVMDispatchNode(BoneName=None, Default=0, Name=deformRig_l_thigh_up_alpha) {template="DISPATCH_RigDispatch_GetAnimAttribute(in Name,in BoneName,in Default,out Value,out Found)"}
   > &RigUnit_ModifyTransforms(
    ItemToModify=((Item=(Type=Bone,Name="dyn_loincloth_front_l_root"),Transform=(Rotation=(X=-0.4226,Y=0,Z=0,W=0.9063),Translation=(X=5,Y=5,Z=0),Scale3D=(X=1,Y=1,Z=1))))
    WeightMinimum=0
    Weight=@Get_Animation_Attribute_3.Value
   )
   > RigVMFunctionReturnNode()
  ]}
 New Function_ContainedGraph {nodes [
   RigVMFunctionEntryNode()
   > &RigUnit_ModifyTransforms(
    ItemToModify="((Item=(Type=Bone,Name=\"dyn_skirt_left_root_l\"),Transform=(Rotation=(X=0,Y=0,Z=-0.024,W=0.9997),Translation=(X=0,Y=0,Z=0),Scale3D=(X=1,Y=1,Z=1))),(Item=(Type=Bone,Name=\"dyn_skirt_left_root\"),Transform=(Rotation=(X=0,Y=0,Z=-0.2672,W=0.9636),Translation=(X=0 ...(truncated)"
    WeightMinimum=0
    Weight=RigVMDispatchNode(False=0, True=1, Condition=RigVMVariableNode() {variable=isCrouch}.Value) {template="DISPATCH_RigVMDispatch_If(in Condition,in True,in False,out Result)"}.Result
   )
   > &RigUnit_ModifyTransforms(
    ItemToModify="((Item=(Type=Bone,Name=\"dyn_skirt_right_root_l\"),Transform=(Rotation=(X=0,Y=-0.1305,Z=0,W=0.9914),Translation=(X=0,Y=0,Z=0),Scale3D=(X=1,Y=1,Z=1))),(Item=(Type=Bone,Name=\"dyn_skirt_right_root\"),Transform=(Rotation=(X=-0.0076,Y=-0.0868,Z=-0.0868,W=0.9924),T ...(truncated)"
    WeightMinimum=0
    Weight=RigVMDispatchNode(False=0, True=1, Condition=RigVMVariableNode() {variable=isCrouch}.Value) {template="DISPATCH_RigVMDispatch_If(in Condition,in True,in False,out Result)"}.Result
   )
   > n1 = &RigUnit_SphericalPoseReader(
    ActiveRegionScaleFactors=(PositiveWidth=4,NegativeWidth=3,PositiveHeight=2.5,NegativeHeight=2)
    ActiveRegionSize=0.1
    Debug=(bDrawDebug=false,bDraw2D=False,bDrawLocalAxes=False,DebugScale=25,DebugSegments=20,DebugThickness=0.25)
    DriverAxis=(X=1,Y=0,Z=0)
    DriverItem=(Type=Bone,Name="thigh_r")
    FalloffRegionScaleFactors=(PositiveWidth=1,NegativeWidth=4,PositiveHeight=1,NegativeHeight=0.5)
    FalloffSize=0.4
    FlipHeightScaling=False
    FlipWidthScaling=true
    OptionalParentItem=(Type=Bone,Name="None")
    RotationOffset=(X=-80,Y=0,Z=-20)
   )
   > n2 = &RigUnit_SphericalPoseReader(
    ActiveRegionScaleFactors=(PositiveWidth=4,NegativeWidth=3,PositiveHeight=2.5,NegativeHeight=2)
    ActiveRegionSize=0.1
    Debug=(bDrawDebug=false,bDraw2D=False,bDrawLocalAxes=False,DebugScale=25,DebugSegments=20,DebugThickness=0.25)
    DriverAxis=(X=-1,Y=0,Z=0)
    DriverItem=(Type=Bone,Name="thigh_l")
    FalloffRegionScaleFactors=(PositiveWidth=1,NegativeWidth=4,PositiveHeight=1,NegativeHeight=0.5)
    FalloffSize=0.4
    FlipHeightScaling=False
    FlipWidthScaling=False
    OptionalParentItem=(Type=Bone,Name="None")
    RotationOffset=(X=100,Y=0,Z=-20)
   )
   > &RigUnit_ModifyTransforms(
    ItemToModify="((Item=(Type=Bone,Name=\"dyn_skirt_left_root_r\"),Transform=(Rotation=(X=0,Y=-0.2164,Z=0,W=0.9763),Translation=(X=0,Y=0,Z=1),Scale3D=(X=1,Y=1,Z=1))),(Item=(Type=Bone,Name=\"dyn_skirt_left_root\"),Transform=(Rotation=(X=0,Y=-0.0872,Z=0,W=0.9962),Translation=(X= ...(truncated)"
    WeightMinimum=0
    Weight=@n1.OutputParam
   )
   > &RigUnit_ModifyTransforms(
    ItemToModify="((Item=(Type=Bone,Name=\"dyn_skirt_right_root_l\"),Transform=(Rotation=(X=0,Y=-0.2164,Z=0,W=0.9763),Translation=(X=0,Y=0,Z=-1),Scale3D=(X=1,Y=1,Z=1))),(Item=(Type=Bone,Name=\"dyn_skirt_right_root_r\"),Transform=(Rotation=(X=0,Y=-0.0436,Z=0,W=0.999),Translation ...(truncated)"
    WeightMinimum=0
    Weight=@n2.OutputParam
   )
   > RigVMFunctionReturnNode()
  ]}
 New Function_ContainedGraph {nodes [
   RigVMFunctionEntryNode()
   > &RigUnit_ModifyTransforms(
    ItemToModify=((Item=(Type=Bone,Name="dyn_beltAcc_front_00"),Transform=(Rotation=(X=0,Y=0,Z=0,W=1),Translation=(X=0,Y=0.5,Z=0),Scale3D=(X=1,Y=1,Z=1))))
    Weight=1
    WeightMinimum=0
   )
   > RigVMFunctionReturnNode()
  ]}
 RigVMFunctionLibrary {nodes [
   Belt Acc FNC = RigVMCollapseNode(ExecuteContext=())
   Belt FNC = RigVMCollapseNode(ExecuteContext=())
   DeltaRotation = RigVMCollapseNode(
    ExecuteContext=()
    Rotation Order=ZYX
    child=(Type=Bone,Name="None")
    outputMinMax=(X=0,Y=1)
    parent=(Type=Bone,Name="None")
    premult=1
    rotationMinMax=(X=0,Y=0)
    twistAxis=(X=1,Y=0,Z=0)
   )
   Loincloth FNC = RigVMCollapseNode(ExecuteContext=())
   SideSkirt FNC = RigVMCollapseNode(ExecuteContext=())
  ]}
 RigVMModel {nodes [
   n1 = &RigUnit_PrepareForExecution()
   addFKCtrl = RigVMFunctionReferenceNode(
    Custom_FKCtrl=()
    Options_FKCtrl="(pelvis_1_#1=false,spine_3_#2=false,clavicle_5_#3=true,arm_7_#4=true,leg_9_#5=true,head_11_46DC15EF465CD29 ...(truncated)"
    Shape_Settings="(pelvis_2_#6=(bVisible=True,Name=\"Hexagon_Thin\",Color=(R=0.0117,G=0.8281,B=1,A=1),Transform=(Rotation=(X=0,Y=-0.7071,Z=0,W=0.7071),Translation=(X=2,Y=0,Z=0),Scale3D=(X=5,Y=5,Z=7))),spine_4_#7=(bVi ...(truncated)"
    ExecuteContext=@n1.ExecutePin
   )
   RigUnit_BeginExecution = &RigUnit_BeginExecution()
   setFKCtrl = RigVMFunctionReferenceNode(setFK_Ctrl=false, ExecuteContext=@RigUnit_BeginExecution.ExecutePin)
   > n2 = /Script/FortniteGame.RigUnit_FortDeformRig(
    Bool=(bUpperArm=true,bElbow=true,bWrist=true,bGroin=true,bGlute=true,bKnee=true,bPec=False,bBicep=False,bLat=False)
    Curve="(bLUpperArmUpAlpha=False,bRUpperArmUpAlpha=False,bLUpperArmFrAlpha=False,bRUpperArmFrAlpha=False,bLUpperArmBkAlpha=False,bRUpperArmBkAlpha=False,bLWristUpAlpha=False,bRWristUpAlpha=False,bLWristDnAlpha=False,bRWristDnAlpha=False,bLKneeAlpha=False,bRKneeAlp ...(truncated)"
    Debug=(bUpperArmDebug=False,bWristDebug=False,bLatDebug=False,bGroinGluteDebug=False)
    Value="(UpperArmUp=(Rotation=(X=0.0,Y=-0.0,Z=0.0,W=1.0),Translation=(X=0.0,Y=0.0,Z=8.0),Scale3D=(X=0.70,Y=1.0,Z=1.0)),UpperArmFwdFr=(Rotation=(X=0.0,Y=-0.0,Z=0.0,W=1.0),Translation=(X=0.0,Y=10.0,Z=0.0),Scale3D=(X=0.50,Y=1.0,Z=1.0)),UpperArmFwdBk=(Rotation=(X=0.0, ...(truncated)"
   ) {is=struct RigVMUnitNode}
   RigVMFunction_Sequence = &RigVMFunction_Sequence(ExecuteContext=@n2.ExecutePin)
   Belt Acc FNC = RigVMFunctionReferenceNode(ExecuteContext=@RigVMFunction_Sequence.A)
   > Loincloth FNC = RigVMFunctionReferenceNode()
   > SideSkirt FNC = RigVMFunctionReferenceNode()
   > Belt FNC = RigVMFunctionReferenceNode()
   n3 = &RigUnit_SphericalPoseReader(
    ActiveRegionScaleFactors=(PositiveWidth=1,NegativeWidth=1,PositiveHeight=1,NegativeHeight=1)
    ActiveRegionSize=0.05
    Debug=(bDrawDebug=false,DebugScale=25,DebugSegments=20,DebugThickness=0.25)
    DriverAxis=(X=1,Y=0,Z=0)
    DriverItem=(Type=Bone,Name="upperarm_l")
    FalloffRegionScaleFactors=(PositiveWidth=1,NegativeWidth=1,PositiveHeight=1,NegativeHeight=1)
    FalloffSize=0.2
    OptionalParentItem=(Type=Bone,Name="clavicle_l")
    RotationOffset=(X=50,Y=-65,Z=0)
    ExecutePin=@RigVMFunction_Sequence.B
   )
   > &RigUnit_ModifyTransforms(
    ItemToModify="((Item=(Type=Bone,Name=\"dyn_bowLoop_left_root_fw\"),Transform=(Rotation=(X=-0.0317,Y=-0.1337,Z=0.0807,W=0.9872),Translation=(X=0,Y=0,Z=-1),Scale3D=(X=1,Y=1,Z=1))),(Item=(Type=Bone,Name=\"dyn_bowString_left_root\"),Transform=(Rotation=(X=0,Y=0,Z=0,W=-1),Transl ...(truncated)"
    WeightMinimum=0
    Weight=@n3.OutputParam
   )
  ]}
]
hierarchy [
 root {loc=<0 0 0>, rot=<0 0 0 1>}
 pelvis {parent=root, loc=<0 -0.5301 98.0571>, rot=<0.0127 0.707 -0.0127 -0.707>}
 spine_01 {parent=pelvis, loc=<6.1208 -0.9328 0>, rot=<0 0 0.034 -0.9994>}
 spine_02 {parent=spine_01, loc=<6.5997 -0.2095 0>, rot=<0 0 0.0238 -0.9997>}
 spine_03 {parent=spine_02, loc=<6.4914 -0.1237 0>, rot=<0 0 0.0149 -0.9999>}
 spine_04 {parent=spine_03, loc=<6.9598 0.0794 0>, rot=<0 0 -0.0132 -0.9999>}
 spine_05 {parent=spine_04, loc=<6.9987 0 0>, rot=<0 0 -0.0015 -1>}
 neck_01 {parent=spine_05, loc=<11.3138 0.1529 0>, rot=<0 0 -0.2003 -0.9797>}
 neck_02 {parent=neck_01, loc=<4.5627 0 0>, rot=<0 0 0.0504 -0.9987>}
 head {parent=neck_02, loc=<5.3914 0 0>, rot=<0 0 0.115 -0.9934>}
 clavicle_l {parent=spine_05, loc=<5.4831 -0.1053 -3.2946>, rot=<0.0465 -0.7289 -0.009 -0.6829>}
 upperarm_l {parent=clavicle_l, loc=<9.8658 -1.2136 -0.4105>, rot=<0.0164 -0.3245 -0.0269 -0.9454>}
 lowerarm_l {parent=upperarm_l, loc=<26.465 -0.8345 0.259>, rot=<0.0036 0.0161 -0.2298 -0.9731>}
 hand_l {parent=lowerarm_l, loc=<22.7355 -0.3203 -0.6105>, rot=<-0.6784 -0.0223 0.0735 -0.7307>}
 thumb_01_l {parent=hand_l, loc=<1.3118 -0.3191 -2.459>, rot=<0.6546 -0.4042 0.0664 -0.6354>}
 thumb_02_l {parent=thumb_01_l, loc=<5.9336 -0.0448 0.0401>, rot=<-0.0037 0.0213 0.298 -0.9543>}
 thumb_03_l {parent=thumb_02_l, loc=<2.2223 0.3682 -0.0907>, rot=<0.0118 -0.013 0.0066 -0.9998>}
 index_metacarpal_l {parent=hand_l, loc=<1.6535 0.0183 -2.96>, rot=<0.0158 -0.0124 -0.0173 -0.9996>}
 index_01_l {parent=index_metacarpal_l, loc=<8.3008 0.0357 -0.0524>, rot=<-0.0044 0.0485 0.0924 -0.9945>}
 index_02_l {parent=index_01_l, loc=<3.3587 -0.3004 0.0452>, rot=<-0.002 0.0027 0.3225 -0.9466>}
 index_03_l {parent=index_02_l, loc=<2.5578 -0.123 -0.0252>, rot=<-0.002 0.0067 0.2558 -0.9667>}
 middle_metacarpal_l {parent=hand_l, loc=<2.1856 -0.1725 -0.5984>, rot=<0.006 0.0168 -0.0329 -0.9993>}
 middle_01_l {parent=middle_metacarpal_l, loc=<7.3275 -0.2447 0.0129>, rot=<-0.0208 0.0042 0.0858 -0.9961>}
 middle_02_l {parent=middle_01_l, loc=<4.0123 0.1401 0.1175>, rot=<-0.0218 0.0258 0.3512 -0.9357>}
 middle_03_l {parent=middle_02_l, loc=<2.6077 0.0534 0.1373>, rot=<0.0082 0.0188 0.1708 -0.9851>}
 ring_metacarpal_l {parent=hand_l, loc=<2.2074 -0.3625 1.3638>, rot=<-0.0284 0.0722 -0.0458 -0.9959>}
 ring_01_l {parent=ring_metacarpal_l, loc=<7.0295 -0.6197 -0.1446>, rot=<-0.0489 -0.0252 0.1892 -0.9804>}
 ring_02_l {parent=ring_01_l, loc=<3.6174 0.4647 -0.1027>, rot=<0 0 0.2806 -0.9598>}
 ring_03_l {parent=ring_02_l, loc=<2.4152 0.114 0.0199>, rot=<0 0 0.1189 -0.9929>}
 pinky_metacarpal_l {parent=hand_l, loc=<1.7616 -0.9753 3.3814>, rot=<-0.0397 0.0814 -0.069 -0.9935>}
 pinky_01_l {parent=pinky_metacarpal_l, loc=<7.1709 -0.4831 -0.1094>, rot=<-0.0654 -0.0035 0.2242 -0.9723>}
 pinky_02_l {parent=pinky_01_l, loc=<2.8274 0.0893 -0.1043>, rot=<0.0037 0.0011 0.2854 -0.9584>}
 pinky_03_l {parent=pinky_02_l, loc=<2.2046 0.0307 -0.0078>, rot=<-0.0002 0.0098 0.1445 -0.9895>}
 deform_wrist_dn_base_l {parent=hand_l, loc=<0 -1.5326 0>, rot=<0 0 0 -1>}
 deform_wrist_dn_l {parent=deform_wrist_dn_base_l, loc=<0 0 0>, rot=<0 0 0 -1>}
 deform_wrist_up_base_l {parent=hand_l, loc=<0 2.3009 0>, rot=<0 0 0 -1>}
 deform_wrist_up_l {parent=deform_wrist_up_base_l, loc=<0 0 0>, rot=<0 0 0 -1>}
 lowerarm_twist_02_l {parent=lowerarm_l, loc=<10.6707 0 0>, rot=<0 0 0 1>}
 lowerarm_twist_01_l {parent=lowerarm_l, loc=<18.316 0 0>, rot=<0 0 0 1>}
 dyn_bracelet_l_MSTR {parent=lowerarm_twist_01_l, loc=<2.4055 0.372 3.7185>, rot=<0.0707 -0.1928 0.0292 -0.9782>}
 deform_elbow_in_base_l {parent=lowerarm_l, loc=<0.1834 3.6062 -0.2572>, rot=<0 0 0 -1>}
 deform_elbow_in_l {parent=deform_elbow_in_base_l, loc=<0 0 0>, rot=<0 0 0 -1>}
 deform_elbow_out_base_l {parent=lowerarm_l, loc=<0 -4 0>, rot=<0 0 0 1>}
 deform_elbow_out_l {parent=deform_elbow_out_base_l, loc=<0 0 0>, rot=<0 0 0 1>}
 upperarm_twist_01_l {parent=upperarm_l, loc=<7.2135 -0.1134 0>, rot=<0 0 0 1>}
 deform_upperarm_bk_base_l {parent=upperarm_twist_01_l, loc=<-6.4053 -4.3746 0.314>, rot=<0 0 0 -1>}
 deform_upperarm_bk_l {parent=deform_upperarm_bk_base_l, loc=<0 0 0>, rot=<0 0 0 -1>}
 deform_upperarm_fr_base_l {parent=upperarm_twist_01_l, loc=<-5.6312 5.092 0.3354>, rot=<0 0 0 -1>}
 deform_upperarm_fr_l {parent=deform_upperarm_fr_base_l, loc=<0 0 0>, rot=<0 0 0 -1>}
 deform_upperarm_up_base_l {parent=upperarm_twist_01_l, loc=<-8.891 0.4825 3.9517>, rot=<0 0 0 -1>}
 deform_upperarm_up_l {parent=deform_upperarm_up_base_l, loc=<0 0 0>, rot=<0 0 0 -1>}
 upperarm_twist_02_l {parent=upperarm_l, loc=<19.7998 -0.1134 0>, rot=<0 0 0 1>}
 clavicle_r {parent=spine_05, loc=<5.4831 -0.1053 3.2946>, rot=<0.7289 0.0465 -0.6829 0.009>}
 upperarm_r {parent=clavicle_r, loc=<-9.8658 1.2136 0.4105>, rot=<0.0164 -0.3245 -0.0269 -0.9454>}
 lowerarm_r {parent=upperarm_r, loc=<-26.465 0.8345 -0.259>, rot=<0.0036 0.0161 -0.2298 -0.9731>}
 hand_r {parent=lowerarm_r, loc=<-22.7355 0.3203 0.6106>, rot=<-0.6784 -0.0223 0.0735 -0.7307>}
 thumb_01_r {parent=hand_r, loc=<-1.3118 0.3191 2.459>, rot=<0.6546 -0.4042 0.0664 -0.6354>}
 thumb_02_r {parent=thumb_01_r, loc=<-5.9336 0.0448 -0.0401>, rot=<-0.0037 0.0213 0.298 -0.9543>}
 thumb_03_r {parent=thumb_02_r, loc=<-2.2223 -0.3682 0.0907>, rot=<0.0118 -0.013 0.0066 -0.9998>}
 index_metacarpal_r {parent=hand_r, loc=<-1.6535 -0.0183 2.96>, rot=<0.0158 -0.0124 -0.0173 -0.9996>}
 index_01_r {parent=index_metacarpal_r, loc=<-8.3008 -0.0357 0.0524>, rot=<-0.0044 0.0485 0.0924 -0.9945>}
 index_02_r {parent=index_01_r, loc=<-3.3587 0.3004 -0.0452>, rot=<-0.002 0.0027 0.3225 -0.9466>}
 index_03_r {parent=index_02_r, loc=<-2.5578 0.123 0.0252>, rot=<-0.002 0.0067 0.2558 -0.9667>}
 middle_metacarpal_r {parent=hand_r, loc=<-2.1856 0.1725 0.5984>, rot=<0.006 0.0168 -0.0329 -0.9993>}
 middle_01_r {parent=middle_metacarpal_r, loc=<-7.3275 0.2447 -0.0129>, rot=<-0.0208 0.0042 0.0858 -0.9961>}
 middle_02_r {parent=middle_01_r, loc=<-4.0123 -0.1401 -0.1175>, rot=<-0.0218 0.0258 0.3512 -0.9357>}
 middle_03_r {parent=middle_02_r, loc=<-2.6077 -0.0534 -0.1373>, rot=<0.0082 0.0188 0.1708 -0.9851>}
 ring_metacarpal_r {parent=hand_r, loc=<-2.2074 0.3625 -1.3638>, rot=<-0.0284 0.0722 -0.0458 -0.9959>}
 ring_01_r {parent=ring_metacarpal_r, loc=<-7.0295 0.6197 0.1446>, rot=<-0.0489 -0.0252 0.1892 -0.9804>}
 ring_02_r {parent=ring_01_r, loc=<-3.6174 -0.4647 0.1027>, rot=<0 0 0.2806 -0.9598>}
 ring_03_r {parent=ring_02_r, loc=<-2.4152 -0.114 -0.0199>, rot=<0 0 0.1189 -0.9929>}
 pinky_metacarpal_r {parent=hand_r, loc=<-1.7616 0.9753 -3.3814>, rot=<-0.0397 0.0814 -0.069 -0.9935>}
 pinky_01_r {parent=pinky_metacarpal_r, loc=<-7.1709 0.4831 0.1094>, rot=<-0.0654 -0.0035 0.2242 -0.9723>}
 pinky_02_r {parent=pinky_01_r, loc=<-2.8274 -0.0893 0.1043>, rot=<0.0037 0.0011 0.2854 -0.9584>}
 pinky_03_r {parent=pinky_02_r, loc=<-2.2046 -0.0307 0.0078>, rot=<-0.0002 0.0098 0.1445 -0.9895>}
 deform_wrist_dn_base_r {parent=hand_r, loc=<-0.0001 1.5328 0>, rot=<0 0 0 -1>}
 deform_wrist_dn_r {parent=deform_wrist_dn_base_r, loc=<0 0 0>, rot=<0 0 0 -1>}
 deform_wrist_up_base_r {parent=hand_r, loc=<-0.0002 -2.3007 -0.0001>, rot=<0 0 0 -1>}
 deform_wrist_up_r {parent=deform_wrist_up_base_r, loc=<0 0 0>, rot=<0 0 0 -1>}
 lowerarm_twist_02_r {parent=lowerarm_r, loc=<-10.6707 0 0>, rot=<0 0 0 -1>}
 lowerarm_twist_01_r {parent=lowerarm_r, loc=<-18.316 0 0>, rot=<0 0 0 -1>}
 deform_elbow_in_base_r {parent=lowerarm_r, loc=<-0.1834 -3.6062 0.2572>, rot=<0 0 0 -1>}
 deform_elbow_in_r {parent=deform_elbow_in_base_r, loc=<0 0 0>, rot=<0 0 0 -1>}
 deform_elbow_out_base_r {parent=lowerarm_r, loc=<0 4 -0.0001>, rot=<0 0 0 -1>}
 deform_elbow_out_r {parent=deform_elbow_out_base_r, loc=<0 0 0>, rot=<0 0 0 1>}
 upperarm_twist_01_r {parent=upperarm_r, loc=<-7.2135 0.1134 0>, rot=<0 0 0 -1>}
 deform_upperarm_bk_base_r {parent=upperarm_twist_01_r, loc=<6.4053 4.3746 -0.314>, rot=<0 0 0 -1>}
 deform_upperarm_bk_r {parent=deform_upperarm_bk_base_r, loc=<0 0 0>, rot=<0 0 0 -1>}
 deform_upperarm_fr_base_r {parent=upperarm_twist_01_r, loc=<5.6312 -5.092 -0.3354>, rot=<0 0 0 -1>}
 deform_upperarm_fr_r {parent=deform_upperarm_fr_base_r, loc=<0 0 0>, rot=<0 0 0 -1>}
 deform_upperarm_up_base_r {parent=upperarm_twist_01_r, loc=<8.891 -0.4825 -3.9517>, rot=<0 0 0 -1>}
 deform_upperarm_up_r {parent=deform_upperarm_up_base_r, loc=<0 0 0>, rot=<0 0 0 -1>}
 upperarm_twist_02_r {parent=upperarm_r, loc=<-19.7998 0.1134 0>, rot=<0 0 0 -1>}
 attach_backpack {parent=spine_05, loc=<-0.123 -13.473 0>, rot=<-0.0093 -0.707 0.0093 -0.707>}
 Backpack_BR {parent=attach_backpack, type=Null, loc=<0 5 0>, rot=<0 0 0 1>}
 attach_cape {parent=spine_05, loc=<9.017 -11.714 0>, rot=<0.0616 -0.7044 -0.0616 -0.7044>}
 cape {parent=attach_cape, type=Null, loc=<0 5 -1>, rot=<0 0 0 1>}
 dyn_bow_left_root {parent=spine_05, loc=<11.4989 -2.4262 -6.1627>, rot=<0.0401 -0.9992 0 0>}
 dyn_bowLoop_left_root_fw {parent=dyn_bow_left_root, loc=<-0.0295 0.2666 0.139>, rot=<0.0024 -0.7864 -0.0618 -0.6146>}
 dyn_bowString_left_root {parent=dyn_bow_left_root, loc=<0.6563 -0.6009 -0.0229>, rot=<-0.4335 0.3357 0.1503 -0.8227>}
 dyn_bowLoop_left_root_back {parent=dyn_bow_left_root, loc=<0.177 -0.807 -0.3701>, rot=<-0.6386 0.3036 0.4073 -0.578>}
 dyn_bow_right_root {parent=spine_05, loc=<11.4923 -2.584 5.9287>, rot=<0.0401 -0.9992 0 0>}
 dyn_bowLoop_right_root_fw {parent=dyn_bow_right_root, loc=<-0.0491 0.4233 -0.373>, rot=<0.7383 0.1392 0.6522 0.1011>}
 dyn_bowLoop_right_root_back {parent=dyn_bow_right_root, loc=<0.0376 -0.6502 0.1361>, rot=<-0.588 0.2796 0.2729 -0.7082>}
 dyn_bowString_right_root {parent=dyn_bow_right_root, loc=<0.1256 -0.8816 -0.1092>, rot=<-0.3535 -0.1406 0.3059 -0.8728>}
 deform_pec_aim_base_l {parent=spine_04, loc=<3.2456 9.3309 -1.2237>, rot=<0.0603 -0.4033 0.3318 -0.8507>}
 deform_pec_aim_l {parent=deform_pec_aim_base_l, loc=<0 0 0>, rot=<0 0 0 1>}
 deform_pec_base_l {parent=deform_pec_aim_l, loc=<9.3665 0 0>, rot=<0 0 0 1>}
 deform_pec_l {parent=deform_pec_base_l, loc=<0 0 0>, rot=<0 0 0 -1>}
 deform_pec_aim_base_r {parent=spine_04, loc=<3.2452 9.3309 1.2237>, rot=<-0.4033 -0.0603 0.8507 0.3318>}
 deform_pec_aim_r {parent=deform_pec_aim_base_r, loc=<0 0 0>, rot=<0 0 0 1>}
 deform_pec_base_r {parent=deform_pec_aim_r, loc=<-9.3667 -0.0003 -0.0004>, rot=<0 0 0 1>}
 deform_pec_r {parent=deform_pec_base_r, loc=<0 0 0>, rot=<0 0 0 -1>}
 thigh_l {parent=pelvis, loc=<-1.0855 -0.2109 -8.7053>, rot=<0.0009 0.0355 0.0155 -0.9993>}
 calf_l {parent=thigh_l, loc=<-42.5284 -0.0019 -0.7297>, rot=<0.0006 -0.0161 -0.0391 -0.9991>}
 foot_l {parent=calf_l, loc=<-46.7401 -2.5014 -0.592>, rot=<0.0427 -0.0238 0.0353 -0.9982>}
 ball_l {parent=foot_l, loc=<-3.5658 10.8066 -0.1544>, rot=<0.0001 -0.0001 -0.7063 -0.7079>}
 calf_twist_02_l {parent=calf_l, loc=<-19.1546 -1.7943 -0.2638>, rot=<0.0237 0.0095 -0.0035 -0.9997>}
 calf_twist_01_l {parent=calf_l, loc=<-35.8483 -1.9197 -0.5794>, rot=<0.0237 0.0095 -0.0035 -0.9997>}
 deform_knee_in_base_l {parent=calf_l, loc=<0.8031 -5.0231 0.2456>, rot=<0 0 0 -1>}
 deform_knee_in_l {parent=deform_knee_in_base_l, loc=<0 0 0>, rot=<0 0 0 -1>}
 deform_knee_out_base_l {parent=calf_l, loc=<0 4 0>, rot=<0 0 0 -1>}
 deform_knee_out_l {parent=deform_knee_out_base_l, loc=<0 0 0>, rot=<0 0 0 1>}
 thigh_twist_01_l {parent=thigh_l, loc=<-27.952 0 0>, rot=<0.0013 0 0 -1>}
 deform_groin_base_l {parent=thigh_l, loc=<0.4801 5.6776 -0.2486>, rot=<0 0 0 -1>}
 deform_groin_l {parent=deform_groin_base_l, loc=<0 0 0>, rot=<0 0 0 -1>}
 cr_thigh_l_no_twist {parent=thigh_l, loc=<0 0 0>, rot=<0 0 0 1>}
 dyn_skirt_left_root_l {parent=thigh_l, loc=<-11.0742 3.8625 -6.6421>, rot=<0.0091 -0.9913 -0.1312 0.0039>}
 dyn_skirt_left_01_l {parent=dyn_skirt_left_root_l, loc=<0.1 0 0>, rot=<0 0 0 1>}
 dyn_skirt_left_02_l {parent=dyn_skirt_left_01_l, loc=<8.2326 0 0>, rot=<0 0 0 1>}
 dyn_skirt_left_03_l {parent=dyn_skirt_left_02_l, loc=<11.2215 0 0>, rot=<0 0 0 1>}
 dyn_skirt_left_04_l {parent=dyn_skirt_left_03_l, loc=<10.9426 0 0>, rot=<0 0 0 1>}
 dyn_skirt_left_root {parent=thigh_l, loc=<-7.3281 -0.9067 -7.5102>, rot=<0.0304 -0.995 0.052 -0.0794>}
 dyn_skirt_left_01 {parent=dyn_skirt_left_root, loc=<0.1 0 0>, rot=<0 0 0 1>}
 dyn_skirt_left_02 {parent=dyn_skirt_left_01, loc=<11.5767 0 -0.8285>, rot=<0 -0.0785 0 -0.9969>}
 dyn_skirt_left_03 {parent=dyn_skirt_left_02, loc=<11.3044 0 0>, rot=<0 0 0 1>}
 dyn_skirt_left_04 {parent=dyn_skirt_left_03, loc=<11.9398 0 0>, rot=<0 0 0 1>}
 dyn_skirt_left_root_r {parent=thigh_l, loc=<-5.3659 -6.8738 -5.5359>, rot=<0.0249 -0.9237 0.3823 0.01>}
 dyn_skirt_left_01_r {parent=dyn_skirt_left_root_r, loc=<0.1 0 0>, rot=<0 0 0 1>}
 dyn_skirt_left_02_r {parent=dyn_skirt_left_01_r, loc=<12.7837 0 0>, rot=<0 0 0 1>}
 dyn_skirt_left_03_r {parent=dyn_skirt_left_02_r, loc=<12.0515 0 0>, rot=<0 0 0 1>}
 dyn_skirt_left_04_r {parent=dyn_skirt_left_03_r, loc=<10.8312 0 0>, rot=<0 0 0 1>}
 thigh_r {parent=pelvis, loc=<-1.0855 -0.2109 8.7053>, rot=<0.0355 -0.0009 0.9993 0.0155>}
 calf_r {parent=thigh_r, loc=<42.5284 0.0019 0.7297>, rot=<0.0006 -0.0161 -0.0391 -0.9991>}
 foot_r {parent=calf_r, loc=<46.7401 2.5014 0.592>, rot=<0.0427 -0.0238 0.0353 -0.9982>}
 ball_r {parent=foot_r, loc=<3.5658 -10.8066 0.1544>, rot=<0.0001 -0.0001 -0.7063 -0.7079>}
 calf_twist_02_r {parent=calf_r, loc=<19.1546 1.7943 0.2638>, rot=<0.0237 0.0095 -0.0035 -0.9997>}
 calf_twist_01_r {parent=calf_r, loc=<35.8483 1.9197 0.5794>, rot=<0.0237 0.0095 -0.0035 -0.9997>}
 deform_knee_in_base_r {parent=calf_r, loc=<-0.8031 5.0231 -0.2457>, rot=<0 0 0 -1>}
 deform_knee_in_r {parent=deform_knee_in_base_r, loc=<0 0 0>, rot=<0 0 0 -1>}
 deform_knee_out_base_r {parent=calf_r, loc=<0 -4 0>, rot=<0 0 0 -1>}
 deform_knee_out_r {parent=deform_knee_out_base_r, loc=<0 0 0>, rot=<0 0 0 1>}
 thigh_twist_01_r {parent=thigh_r, loc=<27.952 0 0>, rot=<0.0013 0 0 -1>}
 deform_groin_base_r {parent=thigh_r, loc=<-0.4801 -5.6776 0.2486>, rot=<0 0 0 -1>}
 deform_groin_r {parent=deform_groin_base_r, loc=<0 0 0>, rot=<0 0 0 -1>}
 cr_thigh_r_no_twist {parent=thigh_r, loc=<0 0 0>, rot=<0 0 0 1>}
 dyn_skirt_right_root_l {parent=thigh_r, loc=<5.3659 6.8738 5.536>, rot=<0.0249 -0.9237 0.3823 0.01>}
 dyn_skirt_right_01_l {parent=dyn_skirt_right_root_l, loc=<-0.1 0 0>, rot=<0 0 0 1>}
 dyn_skirt_right_02_l {parent=dyn_skirt_right_01_l, loc=<-12.7837 0 0>, rot=<0 0 0 1>}
 dyn_skirt_right_03_l {parent=dyn_skirt_right_02_l, loc=<-12.0515 -0.0001 -0.0001>, rot=<0 0 0 1>}
 dyn_skirt_right_04_l {parent=dyn_skirt_right_03_l, loc=<-10.8312 0 0>, rot=<0 0 0 1>}
 dyn_skirt_right_root {parent=thigh_r, loc=<7.3281 0.9067 7.5102>, rot=<0.0304 -0.995 0.052 -0.0794>}
 dyn_skirt_right_01 {parent=dyn_skirt_right_root, loc=<-0.1 0 0>, rot=<0 0 0 -1>}
 dyn_skirt_right_02 {parent=dyn_skirt_right_01, loc=<-11.5767 0 0.8285>, rot=<0 -0.0785 0 -0.9969>}
 dyn_skirt_right_03 {parent=dyn_skirt_right_02, loc=<-11.3044 0 0>, rot=<0 0 0 1>}
 dyn_skirt_right_04 {parent=dyn_skirt_right_03, loc=<-11.9398 0 0.0001>, rot=<0 0 0 1>}
 dyn_skirt_right_root_r {parent=thigh_r, loc=<11.0742 -3.8625 6.6421>, rot=<0.0091 -0.9913 -0.1312 0.0039>}
 dyn_skirt_right_01_r {parent=dyn_skirt_right_root_r, loc=<-0.1 0 0>, rot=<0 0 0 1>}
 dyn_skirt_right_02_r {parent=dyn_skirt_right_01_r, loc=<-8.2327 0 0>, rot=<0 0 0 -1>}
 dyn_skirt_right_03_r {parent=dyn_skirt_right_02_r, loc=<-11.2215 0 0>, rot=<0 0 0 -1>}
 dyn_skirt_right_04_r {parent=dyn_skirt_right_03_r, loc=<-10.9427 0 0>, rot=<0 0 0 -1>}
 dyn_belt_hipAcc_r_MSTR {parent=pelvis, loc=<-1.8169 0.1075 15.9359>, rot=<-0.0179 -0.9998 0 0>}
 dyn_belt_hipAcc_r {parent=dyn_belt_hipAcc_r_MSTR, loc=<0 0 0>, rot=<0 0 0 1>}
 deform_glute_base_l {parent=pelvis, loc=<-7.3466 -9.0213 -9.1245>, rot=<0.0172 0.031 -0.4477 -0.8935>}
 deform_glute_l {parent=deform_glute_base_l, loc=<0 0 0>, rot=<0 0 0 1>}
 deform_glute_base_r {parent=pelvis, loc=<-7.3464 -9.0213 9.1245>, rot=<0.031 -0.0172 0.8935 -0.4477>}
 deform_glute_r {parent=deform_glute_base_r, loc=<0 0 0>, rot=<0 0 0 1>}
 dyn_beltAcc_front_MSTR {parent=pelvis, loc=<12.9299 4.7616 -0.5012>, rot=<-0.2471 -0.969 0 0>}
 dyn_beltAcc_front_00 {parent=dyn_beltAcc_front_MSTR, loc=<0 0 0>, rot=<0 0 0 1>}
 dyn_beltAcc_front_01 {parent=dyn_beltAcc_front_00, loc=<5.3595 0.3217 0>, rot=<0 0 0.2351 -0.972>}
 dyn_beltAcc_back_MSTR {parent=pelvis, loc=<14.3491 -5.178 0>, rot=<0.3545 -0.9351 0.0005 0.0004>}
 dyn_beltAcc_back_00 {parent=dyn_beltAcc_back_MSTR, loc=<0 0 0>, rot=<0 0 0 1>}
 dyn_beltAcc_back_01 {parent=dyn_beltAcc_back_00, loc=<5.5288 0 0>, rot=<0 0 -0.0574 -0.9984>}
 dyn_beltAcc_back_02 {parent=dyn_beltAcc_back_01, loc=<2.5903 0 0>, rot=<0.0004 0.0005 -0.1478 -0.989>}
 dyn_loincloth_rear_root_l {parent=pelvis, loc=<-0.1412 -11.896 -4.3161>, rot=<-0.0148 0 0.9999 0>}
 dyn_loincloth_rear_01_l {parent=dyn_loincloth_rear_root_l, loc=<0.1 0 0>, rot=<0 0 0 1>}
 dyn_loincloth_rear_02_l {parent=dyn_loincloth_rear_01_l, loc=<12.85 0 0>, rot=<0 0 0 1>}
 dyn_loincloth_rear_03_l {parent=dyn_loincloth_rear_02_l, loc=<9.068 0 0>, rot=<0 0 0 1>}
 dyn_loincloth_rear_04_l {parent=dyn_loincloth_rear_03_l, loc=<8.4722 0 0>, rot=<0 0 0 1>}
 dyn_loincloth_rear_root {parent=pelvis, loc=<-0.2738 -11.8963 0>, rot=<0 0 1 0>}
 dyn_loincloth_rear_01 {parent=dyn_loincloth_rear_root, loc=<0.1 0 0>, rot=<0 0 0 1>}
 dyn_loincloth_rear_02 {parent=dyn_loincloth_rear_01, loc=<12.6406 0 0>, rot=<0 0 0 1>}
 dyn_loincloth_rear_03 {parent=dyn_loincloth_rear_02, loc=<9.0678 0 0>, rot=<0 0 0 1>}
 dyn_loincloth_rear_04 {parent=dyn_loincloth_rear_03, loc=<8.4306 0 0>, rot=<0 0 0 1>}
 dyn_loincloth_rear_root_r {parent=pelvis, loc=<-0.1412 -11.896 4.3161>, rot=<0 -0.0148 0 -0.9999>}
 dyn_loincloth_rear_01_r {parent=dyn_loincloth_rear_root_r, loc=<-0.1 0 0>, rot=<0 0 0 1>}
 dyn_loincloth_rear_02_r {parent=dyn_loincloth_rear_01_r, loc=<-12.85 0 0>, rot=<0 0 0 1>}
 dyn_loincloth_rear_03_r {parent=dyn_loincloth_rear_02_r, loc=<-9.068 0 0>, rot=<0 0 0 1>}
 dyn_loincloth_rear_04_r {parent=dyn_loincloth_rear_03_r, loc=<-8.21 0 0>, rot=<0 0 0 1>}
 dyn_loincloth_front_MSTR {parent=pelvis, loc=<4.1735 6.9279 0>, rot=<-0.0179 -0.9998 0 0>}
 dyn_loincloth_front_r_root {parent=dyn_loincloth_front_MSTR, loc=<6.5279 0.0383 -4.4441>, rot=<0.0113 0.0001 0.9999 -0.0046>}
 dyn_loincloth_front_01_r {parent=dyn_loincloth_front_r_root, loc=<-0.1 0 0>, rot=<0 0 0 1>}
 dyn_loincloth_front_02_r {parent=dyn_loincloth_front_01_r, loc=<-10.3771 0 0>, rot=<0 0 0.0044 -1>}
 dyn_loincloth_front_03_r {parent=dyn_loincloth_front_02_r, loc=<-6.5317 0 0>, rot=<0 0 0 1>}
 dyn_loincloth_front_04_r {parent=dyn_loincloth_front_03_r, loc=<-7.7225 0 0>, rot=<0 0 0 1>}
 dyn_loincloth_front_l_root {parent=dyn_loincloth_front_MSTR, loc=<6.5279 0.0383 4.4441>, rot=<-0.0001 0.0113 -0.0046 -0.9999>}
 dyn_loincloth_front_01_l {parent=dyn_loincloth_front_l_root, loc=<0.0999 0 0>, rot=<0 0 0 1>}
 dyn_loincloth_front_02_l {parent=dyn_loincloth_front_01_l, loc=<10.3772 0 0>, rot=<0 0 0.0044 -1>}
 dyn_loincloth_front_03_l {parent=dyn_loincloth_front_02_l, loc=<6.5317 0 0>, rot=<0 0 0 1>}
 dyn_loincloth_front_04_l {parent=dyn_loincloth_front_03_l, loc=<7.7225 0 0>, rot=<0 0 0 1>}
 dyn_loincloth_front_root {parent=dyn_loincloth_front_MSTR, loc=<7.5846 0.0004 0>, rot=<0 0 -0.0046 -1>}
 dyn_loincloth_front_01 {parent=dyn_loincloth_front_root, loc=<0.1 0 0>, rot=<0 0 0 1>}
 dyn_loincloth_front_02 {parent=dyn_loincloth_front_01, loc=<9.4367 0 0>, rot=<0 0 0.0044 -1>}
 dyn_loincloth_front_03 {parent=dyn_loincloth_front_02, loc=<8.7077 0 0>, rot=<0 0 0 1>}
 dyn_loincloth_front_04 {parent=dyn_loincloth_front_03, loc=<9.0093 0 0>, rot=<0 0 0 1>}
]
variables [isCrouch {type=bool, default=False}]
