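# Physics asset for the CherryKnot Strings female medium body cosmetic, setting up physical motion for its front and rear loincloth strips.
# Six strips (left, center, right, front and rear) are four-capsule chains hung from small root bodies, beside pelvis and thigh capsules.
# Each chain swings within tight limits, narrowest at the top joint, with almost no twist.
# Only the front-right root has a constraint to the pelvis, fully locked; the other five roots have none.
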PhysicsAsset /BRCosmetics/Characters/Player/Female/Medium/Bodies/F_MED_CherryKnot_Strings/Meshes/F_MED_CherryKnot_Strings_Physics
bodies {
 dyn_loincloth_front_01 [{shapeType=Capsule, length=6, radius=2.25, center=<3.5 0.5 0>, rotation=<rot 90 0 0>}]
 dyn_loincloth_front_01_l [{shapeType=Capsule, length=6, radius=2.25, center=<3.5 0.25 0.75>, rotation=<rot 90 0 0>}]
 dyn_loincloth_front_01_r [{shapeType=Capsule, length=6, radius=2.25, center=<-3.5 0 -0.75>, rotation=<rot 90 0 0>}]
 dyn_loincloth_front_02 [{shapeType=Capsule, length=5, radius=2.25, center=<3 0 0>, rotation=<rot 90 0 0>}]
 dyn_loincloth_front_02_l [{shapeType=Capsule, length=5, radius=2.25, center=<2 0 0.75>, rotation=<rot 90 0 0>}]
 dyn_loincloth_front_02_r [{shapeType=Capsule, length=5, radius=2.25, center=<-2 0 -0.75>, rotation=<rot 90 0 0>}]
 dyn_loincloth_front_03 [{shapeType=Capsule, length=5, radius=2.25, center=<4 0 0>, rotation=<rot 90 0 0>}]
 dyn_loincloth_front_03_l [{shapeType=Capsule, length=5, radius=2.25, center=<4 0 0.5>, rotation=<rot 90 0 0>}]
 dyn_loincloth_front_03_r [{shapeType=Capsule, length=5, radius=2.25, center=<-4 0 -0.5>, rotation=<rot 90 0 0>}]
 dyn_loincloth_front_04 [{shapeType=Capsule, length=2, radius=2.25, center=<3 0 0>, rotation=<rot 90 0 0>}]
 dyn_loincloth_front_04_l [{shapeType=Capsule, length=2, radius=2.25, center=<3 0 0.5>, rotation=<rot 90 0 0>}]
 dyn_loincloth_front_04_r [{shapeType=Capsule, length=2, radius=2.25, center=<-3 0 -0.5>, rotation=<rot 90 0 0>}]
 dyn_loincloth_front_l_root [{shapeType=Capsule, length=0.505, radius=0.505, shapeName=dyn_loincloth_front_l_root_capsule, rotation=<rot 0 -90 0>}]
 dyn_loincloth_front_r_root [{shapeType=Capsule, length=0.505, radius=0.505, shapeName=dyn_loincloth_front_r_root_capsule, rotation=<rot 0 -90 0>}]
 dyn_loincloth_front_root [{shapeType=Capsule, length=0.505, radius=0.505, shapeName=dyn_loincloth_front_root_capsule, rotation=<rot 0 -90 0>}]
 dyn_loincloth_rear_01 [{shapeType=Capsule, length=9, radius=2.25, shapeName=dyn_loincloth_rear_01_capsule, center=<6 1.5 0>, rotation=<rot 90 0 0>}]
 dyn_loincloth_rear_01_l [{shapeType=Capsule, length=9, radius=2.25, shapeName=dyn_loincloth_rear_01_l_capsule, center=<6 1.5 0>, rotation=<rot 90 0 0>}]
 dyn_loincloth_rear_01_r [{shapeType=Capsule, length=9, radius=2.25, shapeName=dyn_loincloth_rear_01_r_capsule, center=<-6 -1.5 0>, rotation=<rot 90 0 0>}]
 dyn_loincloth_rear_02 [{shapeType=Capsule, length=5, radius=2.25, shapeName=dyn_loincloth_rear_02_capsule, center=<4 0 0>, rotation=<rot 90 0 0>}]
 dyn_loincloth_rear_02_l [{shapeType=Capsule, length=5, radius=2.25, shapeName=dyn_loincloth_rear_02_l_capsule, center=<4 0 0>, rotation=<rot 90 0 0>}]
 dyn_loincloth_rear_02_r [{shapeType=Capsule, length=5, radius=2.25, shapeName=dyn_loincloth_rear_02_r_capsule, center=<-4 0 0>, rotation=<rot 90 0 0>}]
 dyn_loincloth_rear_03 [{shapeType=Capsule, length=5, radius=2.25, shapeName=dyn_loincloth_rear_03_capsule, center=<4 0 0>, rotation=<rot 90 0 0>}]
 dyn_loincloth_rear_03_l [{shapeType=Capsule, length=5, radius=2.25, shapeName=dyn_loincloth_rear_03_l_capsule, center=<4 0 0>, rotation=<rot 90 0 0>}]
 dyn_loincloth_rear_03_r [{shapeType=Capsule, length=5, radius=2.25, shapeName=dyn_loincloth_rear_03_r_capsule, center=<-4 0 0>, rotation=<rot 90 0 0>}]
 dyn_loincloth_rear_04 [{shapeType=Capsule, length=6, radius=2.25, shapeName=dyn_loincloth_rear_04_capsule, center=<4 0 0>, rotation=<rot 90 0 0>}]
 dyn_loincloth_rear_04_l [{shapeType=Capsule, length=6, radius=2.25, shapeName=dyn_loincloth_rear_04_l_capsule, center=<4 0 0>, rotation=<rot 90 0 0>}]
 dyn_loincloth_rear_04_r [{shapeType=Capsule, length=6, radius=2.25, shapeName=dyn_loincloth_rear_04_r_capsule, center=<-4 0 0>, rotation=<rot 90 0 0>}]
 dyn_loincloth_rear_root [{shapeType=Capsule, length=0.505, radius=0.505, shapeName=dyn_loincloth_rear_root_capsule, rotation=<rot 0 -90 0>}]
 dyn_loincloth_rear_root_l [{shapeType=Capsule, length=0.505, radius=0.505, shapeName=dyn_loincloth_rear_root_l_capsule, rotation=<rot 0 -90 0>}]
 dyn_loincloth_rear_root_r [{shapeType=Capsule, length=0.505, radius=0.505, shapeName=dyn_loincloth_rear_root_r_capsule, rotation=<rot 0 -90 0>}]
 pelvis [{shapeType=Capsule, length=13.7796, radius=8, center=<-1.8476 -1.4816 0>}]
 thigh_l [
  {shapeType=Capsule, length=22.5, radius=7.5, center=<-10.5461 -0.5061 -0.0853>, rotation=<rot 0 -92 -92>}
  {shapeType=Capsule, length=8.7, radius=6, shapeName=thigh_l_capsule, center=<-30 0.2 -0.5>, rotation=<rot 90 -30 -30>}
 ]
 thigh_r [
  {shapeType=Capsule, length=22.5, radius=7.5, center=<10.5461 0.5061 0.0853>, rotation=<rot 0 -92 88>}
  {shapeType=Capsule, length=8.7, radius=6, shapeName=thigh_r_capsule, center=<30 -0.2 -0.5>, rotation=<rot 90 -30 150>}
 ]
}
constraints [
 {bone1Name=dyn_loincloth_front_01_l, bone2Name=dyn_loincloth_front_l_root, swing1LimitDegrees=5, swing2LimitDegrees=5, twistLimitDegrees=5}
 {bone1Name=dyn_loincloth_front_02_l, bone2Name=dyn_loincloth_front_01_l, swing1LimitDegrees=15, swing2LimitDegrees=15, twistLimitDegrees=0}
 {bone1Name=dyn_loincloth_front_03_l, bone2Name=dyn_loincloth_front_02_l, swing1LimitDegrees=15, swing2LimitDegrees=15, twistLimitDegrees=0}
 {bone1Name=dyn_loincloth_front_04_l, bone2Name=dyn_loincloth_front_03_l, swing1LimitDegrees=15, swing2LimitDegrees=15, twistLimitDegrees=0}
 {
  bone1Name=dyn_loincloth_front_r_root
  bone2Name=pelvis
  swing1LimitDegrees=45
  swing1Motion=Locked
  swing2LimitDegrees=45
  swing2Motion=Locked
  twistLimitDegrees=45
  twistMotion=Locked
 }
 {bone1Name=dyn_loincloth_front_01_r, bone2Name=dyn_loincloth_front_r_root, swing1LimitDegrees=5, swing2LimitDegrees=5, twistLimitDegrees=5}
 {bone1Name=dyn_loincloth_front_02_r, bone2Name=dyn_loincloth_front_01_r, swing1LimitDegrees=15, swing2LimitDegrees=15, twistLimitDegrees=0}
 {bone1Name=dyn_loincloth_front_03_r, bone2Name=dyn_loincloth_front_02_r, swing1LimitDegrees=15, swing2LimitDegrees=15, twistLimitDegrees=0}
 {bone1Name=dyn_loincloth_front_04_r, bone2Name=dyn_loincloth_front_03_r, swing1LimitDegrees=15, swing2LimitDegrees=15, twistLimitDegrees=0}
 {bone1Name=dyn_loincloth_front_01, bone2Name=dyn_loincloth_front_root, swing1LimitDegrees=5, swing2LimitDegrees=5, twistLimitDegrees=5}
 {bone1Name=dyn_loincloth_front_02, bone2Name=dyn_loincloth_front_01, swing1LimitDegrees=15, swing2LimitDegrees=15, twistLimitDegrees=0}
 {bone1Name=dyn_loincloth_front_03, bone2Name=dyn_loincloth_front_02, swing1LimitDegrees=15, swing2LimitDegrees=15, twistLimitDegrees=0}
 {bone1Name=dyn_loincloth_front_04, bone2Name=dyn_loincloth_front_03, swing1LimitDegrees=15, swing2LimitDegrees=15, twistLimitDegrees=0}
 {bone1Name=dyn_loincloth_rear_01, bone2Name=dyn_loincloth_rear_root, swing1LimitDegrees=12, swing2LimitDegrees=12, twistLimitDegrees=2}
 {bone1Name=dyn_loincloth_rear_02, bone2Name=dyn_loincloth_rear_01, swing1LimitDegrees=15, swing2LimitDegrees=15, twistLimitDegrees=0}
 {bone1Name=dyn_loincloth_rear_03, bone2Name=dyn_loincloth_rear_02, swing1LimitDegrees=15, swing2LimitDegrees=15, twistLimitDegrees=0}
 {bone1Name=dyn_loincloth_rear_04, bone2Name=dyn_loincloth_rear_03, swing1LimitDegrees=15, swing2LimitDegrees=15, twistLimitDegrees=0}
 {bone1Name=dyn_loincloth_rear_01_l, bone2Name=dyn_loincloth_rear_root_l, swing1LimitDegrees=12, swing2LimitDegrees=12, twistLimitDegrees=2}
 {bone1Name=dyn_loincloth_rear_02_l, bone2Name=dyn_loincloth_rear_01_l, swing1LimitDegrees=15, swing2LimitDegrees=15, twistLimitDegrees=0}
 {bone1Name=dyn_loincloth_rear_03_l, bone2Name=dyn_loincloth_rear_02_l, swing1LimitDegrees=15, swing2LimitDegrees=15, twistLimitDegrees=0}
 {bone1Name=dyn_loincloth_rear_04_l, bone2Name=dyn_loincloth_rear_03_l, swing1LimitDegrees=15, swing2LimitDegrees=15, twistLimitDegrees=0}
 {bone1Name=dyn_loincloth_rear_01_r, bone2Name=dyn_loincloth_rear_root_r, swing1LimitDegrees=12, swing2LimitDegrees=12, twistLimitDegrees=2}
 {bone1Name=dyn_loincloth_rear_02_r, bone2Name=dyn_loincloth_rear_01_r, swing1LimitDegrees=15, swing2LimitDegrees=15, twistLimitDegrees=0}
 {bone1Name=dyn_loincloth_rear_03_r, bone2Name=dyn_loincloth_rear_02_r, swing1LimitDegrees=15, swing2LimitDegrees=15, twistLimitDegrees=0}
 {bone1Name=dyn_loincloth_rear_04_r, bone2Name=dyn_loincloth_rear_03_r, swing1LimitDegrees=15, swing2LimitDegrees=15, twistLimitDegrees=0}
]
