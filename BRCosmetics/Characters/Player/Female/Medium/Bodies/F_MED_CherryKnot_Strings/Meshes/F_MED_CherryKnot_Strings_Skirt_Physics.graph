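# Physics asset for the skirt of the female medium CherryKnot Strings body, with capsule bodies for dynamic skirt bones and legs.
# Six skirt chains, three per side, each run from a tiny root body through four segments.
# Each root is rotationally locked to its side's thigh, and the segments are jointed in sequence with uniform swing and twist settings.
# Thigh, calf and foot capsules provide leg volumes, with no constraints between them.
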PhysicsAsset /BRCosmetics/Characters/Player/Female/Medium/Bodies/F_MED_CherryKnot_Strings/Meshes/F_MED_CherryKnot_Strings_Skirt_Physics
bodies {
 calf_l [{shapeType=Capsule, length=40, radius=5.1858, shapeName=calf_l_capsule, center=<-15.5547 -2.5536 -0.4633>, rotation=<rot -0.528 90.8862 -88>}]
 calf_r [{shapeType=Capsule, length=40, radius=4.6858, shapeName=calf_r_capsule, center=<15.5547 2.5536 0.4633>, rotation=<rot -0.528 90.8862 92>}]
 dyn_skirt_left_01 [{shapeType=Capsule, length=7, radius=1.5, shapeName=dyn_skirt_left_01_capsule, center=<4.6164 0 0>, rotation=<rot 0 -81 -81.3321>}]
 dyn_skirt_left_01_l [{shapeType=Capsule, length=7, radius=1.5, shapeName=dyn_skirt_left_01_l_capsule, center=<0.8395 0 0>, rotation=<rot 0 83.2671 -86.4887>}]
 dyn_skirt_left_01_r [{
   shapeType=Capsule
   length=7
   radius=2
   shapeName=dyn_skirt_left_01_r_capsule
   center=<5.366 0.386 -0.4857>
   rotation=<rot 0 97.1897 -89.4877>
  }]
 dyn_skirt_left_02 [{shapeType=Capsule, length=7, radius=2, shapeName=dyn_skirt_left_02_capsule, center=<3 0 0>, rotation=<rot 0 -90 -90>}]
 dyn_skirt_left_02_l [{shapeType=Capsule, length=7, radius=1.5, shapeName=dyn_skirt_left_02_l_capsule, center=<3.5037 0 0.5>, rotation=<rot 0 -90 -90>}]
 dyn_skirt_left_02_r [{shapeType=Capsule, length=7, radius=2, shapeName=dyn_skirt_left_02_r_capsule, center=<2.5 0 0>, rotation=<rot 0 -90 -90>}]
 dyn_skirt_left_03 [{shapeType=Capsule, length=7, radius=2, shapeName=dyn_skirt_left_03_capsule, center=<3 0 0>, rotation=<rot 0 -90 -90>}]
 dyn_skirt_left_03_l [{shapeType=Capsule, length=7, radius=1.5, shapeName=dyn_skirt_left_03_l_capsule, center=<2.8116 0 0>, rotation=<rot 0 -90 -90>}]
 dyn_skirt_left_03_r [{shapeType=Capsule, length=7, radius=2, shapeName=dyn_skirt_left_03_r_capsule, center=<2 0 0>, rotation=<rot 0 -90 -90>}]
 dyn_skirt_left_04 [{shapeType=Capsule, length=7, radius=2, shapeName=dyn_skirt_left_04_capsule, center=<2 0 0>, rotation=<rot 0 -90 -90>}]
 dyn_skirt_left_04_l [{shapeType=Capsule, length=7, radius=2, shapeName=dyn_skirt_left_04_l_capsule, center=<2.5273 0 0>, rotation=<rot 0 -90 -90>}]
 dyn_skirt_left_04_r [{shapeType=Capsule, length=7, radius=2, shapeName=dyn_skirt_left_04_r_capsule, center=<2 0 0>, rotation=<rot 0 -90 -90>}]
 dyn_skirt_left_root [{shapeType=Capsule, length=0.505, radius=0.505, shapeName=dyn_skirt_left_root_capsule, rotation=<rot 0 -90 0>}]
 dyn_skirt_left_root_l [{shapeType=Capsule, length=0.505, radius=0.505, shapeName=dyn_skirt_left_root_l_capsule, rotation=<rot 0 -90 0>}]
 dyn_skirt_left_root_r [{shapeType=Capsule, length=0.505, radius=0.505, shapeName=dyn_skirt_left_root_r_capsule, rotation=<rot 0 -90 0>}]
 dyn_skirt_right_01 [{shapeType=Capsule, length=7, radius=1.5, shapeName=dyn_skirt_right_01_capsule, center=<-4.1385 0 0>, rotation=<rot 0 -81 -81.3321>}]
 dyn_skirt_right_01_l [{
   shapeType=Capsule
   length=7
   radius=2
   shapeName=dyn_skirt_right_01_l_capsule
   center=<-4.5835 -0.2874 0.4927>
   rotation=<rot 0 97.1897 -89.4877>
  }]
 dyn_skirt_right_01_r [{shapeType=Capsule, length=7, radius=1.5, shapeName=dyn_skirt_right_01_r_capsule, center=<-0.9223 0 0>, rotation=<rot 0 83.2671 -86.4887>}]
 dyn_skirt_right_02 [{shapeType=Capsule, length=7, radius=2, shapeName=dyn_skirt_right_02_capsule, center=<-3.69 0 0>, rotation=<rot 0 -90 -90>}]
 dyn_skirt_right_02_l [{shapeType=Capsule, length=7, radius=2, shapeName=dyn_skirt_right_02_l_capsule, center=<-2.5 0 0>, rotation=<rot 0 -90 -90>}]
 dyn_skirt_right_02_r [{shapeType=Capsule, length=7, radius=1.5, shapeName=dyn_skirt_right_02_r_capsule, center=<-3.5205 0 -0.5>, rotation=<rot 0 -90 -90>}]
 dyn_skirt_right_03 [{shapeType=Capsule, length=7, radius=2, shapeName=dyn_skirt_right_03_capsule, center=<-3.1449 0 0>, rotation=<rot 0 -90 -90>}]
 dyn_skirt_right_03_l [{shapeType=Capsule, length=7, radius=2, shapeName=dyn_skirt_right_03_l_capsule, center=<-2 0 0>, rotation=<rot 0 -90 -90>}]
 dyn_skirt_right_03_r [{shapeType=Capsule, length=7, radius=1.5, shapeName=dyn_skirt_right_03_r_capsule, center=<-3 0 0>, rotation=<rot 0 -90 -90>}]
 dyn_skirt_right_04 [{shapeType=Capsule, length=7, radius=2, shapeName=dyn_skirt_right_04_capsule, center=<-2.0444 0 0>, rotation=<rot 0 -90 -90>}]
 dyn_skirt_right_04_l [{shapeType=Capsule, length=7, radius=2, shapeName=dyn_skirt_right_04_l_capsule, center=<-2 0 0>, rotation=<rot 0 -90 -90>}]
 dyn_skirt_right_04_r [{shapeType=Capsule, length=7, radius=2, shapeName=dyn_skirt_right_04_r_capsule, center=<-2 0 0>, rotation=<rot 0 -90 -90>}]
 dyn_skirt_right_root [{shapeType=Capsule, length=0.505, radius=0.505, shapeName=dyn_skirt_right_root_capsule, rotation=<rot 0 -90 0>}]
 dyn_skirt_right_root_l [{shapeType=Capsule, length=0.505, radius=0.505, shapeName=dyn_skirt_right_root_l_capsule, rotation=<rot 0 -90 0>}]
 dyn_skirt_right_root_r [{shapeType=Capsule, length=0.505, radius=0.505, shapeName=dyn_skirt_right_root_r_capsule, rotation=<rot 0 -90 0>}]
 foot_l [{
   shapeType=Capsule
   length=12.3754
   radius=8.7644
   shapeName=foot_l_capsule
   center=<-3.7594 6.8434 -0.1707>
   rotation=<rot 0 24.2527 -86.8418>
  }]
 foot_r [{
   shapeType=Capsule
   length=11.9363
   radius=8.0139
   shapeName=foot_r_capsule
   center=<2.5459 -7.5647 0.3782>
   rotation=<rot 0 11.1125 -88.7337>
  }]
 thigh_l [{shapeType=Capsule, length=22, radius=6.25, center=<-20.8852 -0.2234 -0.4467>, rotation=<rot 0 -92 -92>}]
 thigh_r [{shapeType=Capsule, length=22, radius=6.25, center=<20.8852 0.2234 0.4467>, rotation=<rot 0 -92 88>}]
}
constraints [
 {
  bone1Name=dyn_skirt_left_root
  bone2Name=thigh_l
  swing1LimitDegrees=45
  swing1Motion=Locked
  swing2LimitDegrees=45
  swing2Motion=Locked
  twistLimitDegrees=45
  twistMotion=Locked
 }
 {bone1Name=dyn_skirt_left_02, bone2Name=dyn_skirt_left_01, swing1LimitDegrees=25, swing2LimitDegrees=25, twistLimitDegrees=5}
 {bone1Name=dyn_skirt_left_03, bone2Name=dyn_skirt_left_02, swing1LimitDegrees=25, swing2LimitDegrees=25, twistLimitDegrees=5}
 {bone1Name=dyn_skirt_left_04, bone2Name=dyn_skirt_left_03, swing1LimitDegrees=25, swing2LimitDegrees=25, twistLimitDegrees=5}
 {
  bone1Name=dyn_skirt_left_root_l
  bone2Name=thigh_l
  swing1LimitDegrees=45
  swing1Motion=Locked
  swing2LimitDegrees=45
  swing2Motion=Locked
  twistLimitDegrees=45
  twistMotion=Locked
 }
 {bone1Name=dyn_skirt_left_02_l, bone2Name=dyn_skirt_left_01_l, swing1LimitDegrees=25, swing2LimitDegrees=25, twistLimitDegrees=5}
 {bone1Name=dyn_skirt_left_03_l, bone2Name=dyn_skirt_left_02_l, swing1LimitDegrees=25, swing2LimitDegrees=25, twistLimitDegrees=5}
 {bone1Name=dyn_skirt_left_04_l, bone2Name=dyn_skirt_left_03_l, swing1LimitDegrees=25, swing2LimitDegrees=25, twistLimitDegrees=5}
 {
  bone1Name=dyn_skirt_left_root_r
  bone2Name=thigh_l
  swing1LimitDegrees=45
  swing1Motion=Locked
  swing2LimitDegrees=45
  swing2Motion=Locked
  twistLimitDegrees=45
  twistMotion=Locked
 }
 {bone1Name=dyn_skirt_left_02_r, bone2Name=dyn_skirt_left_01_r, swing1LimitDegrees=25, swing2LimitDegrees=25, twistLimitDegrees=5}
 {bone1Name=dyn_skirt_left_03_r, bone2Name=dyn_skirt_left_02_r, swing1LimitDegrees=25, swing2LimitDegrees=25, twistLimitDegrees=5}
 {bone1Name=dyn_skirt_left_04_r, bone2Name=dyn_skirt_left_03_r, swing1LimitDegrees=25, swing2LimitDegrees=25, twistLimitDegrees=5}
 {
  bone1Name=dyn_skirt_right_root
  bone2Name=thigh_r
  swing1LimitDegrees=45
  swing1Motion=Locked
  swing2LimitDegrees=45
  swing2Motion=Locked
  twistLimitDegrees=45
  twistMotion=Locked
 }
 {bone1Name=dyn_skirt_right_02, bone2Name=dyn_skirt_right_01, swing1LimitDegrees=25, swing2LimitDegrees=25, twistLimitDegrees=5}
 {bone1Name=dyn_skirt_right_03, bone2Name=dyn_skirt_right_02, swing1LimitDegrees=25, swing2LimitDegrees=25, twistLimitDegrees=5}
 {bone1Name=dyn_skirt_right_04, bone2Name=dyn_skirt_right_03, swing1LimitDegrees=25, swing2LimitDegrees=25, twistLimitDegrees=5}
 {
  bone1Name=dyn_skirt_right_root_l
  bone2Name=thigh_r
  swing1LimitDegrees=45
  swing1Motion=Locked
  swing2LimitDegrees=45
  swing2Motion=Locked
  twistLimitDegrees=45
  twistMotion=Locked
 }
 {bone1Name=dyn_skirt_right_02_l, bone2Name=dyn_skirt_right_01_l, swing1LimitDegrees=25, swing2LimitDegrees=25, twistLimitDegrees=5}
 {bone1Name=dyn_skirt_right_03_l, bone2Name=dyn_skirt_right_02_l, swing1LimitDegrees=25, swing2LimitDegrees=25, twistLimitDegrees=5}
 {bone1Name=dyn_skirt_right_04_l, bone2Name=dyn_skirt_right_03_l, swing1LimitDegrees=25, swing2LimitDegrees=25, twistLimitDegrees=5}
 {
  bone1Name=dyn_skirt_right_root_r
  bone2Name=thigh_r
  swing1LimitDegrees=45
  swing1Motion=Locked
  swing2LimitDegrees=45
  swing2Motion=Locked
  twistLimitDegrees=45
  twistMotion=Locked
 }
 {bone1Name=dyn_skirt_right_02_r, bone2Name=dyn_skirt_right_01_r, swing1LimitDegrees=25, swing2LimitDegrees=25, twistLimitDegrees=5}
 {bone1Name=dyn_skirt_right_03_r, bone2Name=dyn_skirt_right_02_r, swing1LimitDegrees=25, swing2LimitDegrees=25, twistLimitDegrees=5}
 {bone1Name=dyn_skirt_right_04_r, bone2Name=dyn_skirt_right_03_r, swing1LimitDegrees=25, swing2LimitDegrees=25, twistLimitDegrees=5}
 {bone1Name=dyn_skirt_left_01, bone2Name=dyn_skirt_left_root, swing1LimitDegrees=25, swing2LimitDegrees=25, twistLimitDegrees=5}
 {bone1Name=dyn_skirt_left_01_l, bone2Name=dyn_skirt_left_root_l, swing1LimitDegrees=25, swing2LimitDegrees=25, twistLimitDegrees=5}
 {bone1Name=dyn_skirt_left_01_r, bone2Name=dyn_skirt_left_root_r, swing1LimitDegrees=25, swing2LimitDegrees=25, twistLimitDegrees=5}
 {bone1Name=dyn_skirt_right_01, bone2Name=dyn_skirt_right_root, swing1LimitDegrees=25, swing2LimitDegrees=25, twistLimitDegrees=5}
 {bone1Name=dyn_skirt_right_01_l, bone2Name=dyn_skirt_right_root_l, swing1LimitDegrees=25, swing2LimitDegrees=25, twistLimitDegrees=5}
 {bone1Name=dyn_skirt_right_01_r, bone2Name=dyn_skirt_right_root_r, swing1LimitDegrees=25, swing2LimitDegrees=25, twistLimitDegrees=5}
]
