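# Post-animation rig for the CherryKnot Strings female body cosmetic, adding secondary motion to its bracelet, belt accessories, bows and skirt.
# It copies the attached parent's pose through a crouch-aware Control Rig, then chains spring-driven AnimDynamics, weighted bow-root constraints and two pelvis-space RigidBody nodes with separate physics assets, one for the skirt.
# Each update computes gravity override and RigidBody settings from a parameter asset, plus a front-end-dependent alpha.
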AnimBlueprint /BRCosmetics/Characters/Player/Female/Medium/Bodies/F_MED_CherryKnot_Strings/Meshes/F_MED_CherryKnot_Strings_AnimBP
target_skeleton=&/F_MED_CherryKnot_Strings_Skeleton
graphs [
 AnimGraph {nodes [
   n1 = Control Rig!(Source=Copy Pose From Mesh() {bUseAttachedParent=true}, Alpha, isCrouch:bool=Get bIsCrouching()) -> (Pose) {controlRigAssetReference {blueprintRigClass=&/F_MED_CherryKnot_Strings_CtrlRig.F_MED_CherryKnot_Strings_CtrlRig_C}}
   n2 = Anim Dynamics - Bone: dyn_bracelet_l_MSTR!(ComponentPose=Local To Component(LocalPose=@n1), Alpha=1, GravityOverride:Vector="0,0,-980") -> (Pose) {
    angularDampingOverride=1
    angularSpringConstant=1000
    bAngularSpring=true
    bOverrideAngularDamping=true
    bUseGravityOverride=true
    gravityScale=0.35
    simulationSpace=BoneRelative
    boundBone {boneName=dyn_bracelet_l_MSTR}
    componentAppliedLinearAccClamp=<3000 3000 3000>
    componentLinearAccScale=<0.5 0.5 0.5>
    componentLinearVelScale=<0.5 0.5 0.5>
    gravityOverride {z=-980}
    physicsBodyDefinitions [{
      boundBone {boneName=dyn_bracelet_l_MSTR}
      boxExtents=<15 17 17>
      constraintSetup {angularLimitsMax=<5 10 10>, angularLimitsMin=<-5 -50 -10>, angularTarget=<1 0 0>}
      localJointOffset=<4 0 -1>
     }]
    relativeSpaceBone {boneName=lowerarm_l}
    simSpaceSettings {maxAngularAcceleration=100, maxAngularVelocity=10, simSpaceAngularAlpha=1}
   }
   n3 = Anim Dynamics - Bone: dyn_beltAcc_front_01!(ComponentPose=@n2, Alpha=1, GravityOverride:Vector="0,0,-980") -> (Pose) {
    angularDampingOverride=1
    angularSpringConstant=3000
    bAngularSpring=true
    bOverrideAngularDamping=true
    bUseGravityOverride=true
    gravityScale=0.35
    simulationSpace=RootRelative
    boundBone {boneName=dyn_beltAcc_front_01}
    componentAppliedLinearAccClamp=<3000 3000 3000>
    componentLinearAccScale=<0.5 0.5 0.5>
    componentLinearVelScale=<0.5 0.5 0.5>
    gravityOverride {z=-980}
    physicsBodyDefinitions [{
      boundBone {boneName=dyn_beltAcc_front_01}
      boxExtents=<15 15 15>
      constraintSetup {angularLimitsMax=<0 4 45>, angularLimitsMin=<0 -4 0>, angularTarget=<1 0 0>}
      localJointOffset=<4 0 0>
     }]
    simSpaceSettings {maxAngularAcceleration=100, maxAngularVelocity=10, simSpaceAngularAlpha=1}
   }
   n4 = Anim Dynamics - Bone: dyn_beltAcc_back_00!(ComponentPose=@n3, Alpha=0.7, GravityOverride:Vector="0,0,-980") -> (Pose) {
    angularDampingOverride=1
    angularSpringConstant=1000
    bAngularSpring=true
    bOverrideAngularDamping=true
    bUseGravityOverride=true
    gravityScale=0.35
    simulationSpace=RootRelative
    boundBone {boneName=dyn_beltAcc_back_00}
    componentAppliedLinearAccClamp=<3000 3000 3000>
    componentLinearAccScale=<0.5 0.5 0.5>
    componentLinearVelScale=<0.5 0.5 0.5>
    gravityOverride {z=-980}
    physicsBodyDefinitions [{
      boundBone {boneName=dyn_beltAcc_back_00}
      boxExtents=<15 15 15>
      constraintSetup {angularLimitsMax=<5 5 0>, angularLimitsMin=<-5 -5 -10>, angularTarget=<1 0 0>}
      localJointOffset=<4 0 0>
     }]
    relativeSpaceBone {boneName=pelvis}
    simSpaceSettings {maxAngularAcceleration=100, maxAngularVelocity=0.7, simSpaceAngularAlpha=1}
   }
   n5 = Anim Dynamics - Bone: dyn_beltAcc_back_01!(ComponentPose=@n4, Alpha=0.7, GravityOverride:Vector="0,0,-980") -> (Pose) {
    angularDampingOverride=1
    angularSpringConstant=1000
    bAngularSpring=true
    bOverrideAngularDamping=true
    bUseGravityOverride=true
    gravityScale=0.35
    simulationSpace=RootRelative
    boundBone {boneName=dyn_beltAcc_back_01}
    componentAppliedLinearAccClamp=<3000 3000 3000>
    componentLinearAccScale=<0.5 0.5 0.5>
    componentLinearVelScale=<0.5 0.5 0.5>
    gravityOverride {z=-980}
    physicsBodyDefinitions [{
      boundBone {boneName=dyn_beltAcc_back_01}
      boxExtents=<15 15 15>
      constraintSetup {angularLimitsMax=<5 5 0>, angularLimitsMin=<-5 -5 -10>, angularTarget=<1 0 0>}
      localJointOffset=<4 0 0>
     }]
    relativeSpaceBone {boneName=pelvis}
    simSpaceSettings {maxAngularAcceleration=100, maxAngularVelocity=0.7, simSpaceAngularAlpha=1}
   }
   n6 = Anim Dynamics - Bone: dyn_beltAcc_back_02!(ComponentPose=@n5, Alpha=0.7, GravityOverride:Vector="0,0,-980") -> (Pose) {
    angularDampingOverride=1
    angularSpringConstant=2000
    bAngularSpring=true
    bOverrideAngularDamping=true
    bUseGravityOverride=true
    gravityScale=0.35
    simulationSpace=RootRelative
    boundBone {boneName=dyn_beltAcc_back_02}
    componentAppliedLinearAccClamp=<3000 3000 3000>
    componentLinearAccScale=<0.5 0.5 0.5>
    componentLinearVelScale=<0.5 0.5 0.5>
    gravityOverride {z=-980}
    physicsBodyDefinitions [{
      boundBone {boneName=dyn_beltAcc_back_02}
      boxExtents=<15 15 15>
      constraintSetup {angularLimitsMax=<5 5 0>, angularLimitsMin=<-5 -5 -50>, angularTarget=<1 0 0>}
      localJointOffset=<4 0 0>
     }]
    relativeSpaceBone {boneName=pelvis}
    simSpaceSettings {maxAngularAcceleration=100, maxAngularVelocity=0.7, simSpaceAngularAlpha=1}
   }
   n7 = "Constraint - dyn_bow_left_root (PPPPPP)"!(
    ComponentPose=@n6
    Alpha=1
    ConstraintWeights_0=0.4039
    ConstraintWeights_1:float=0.3181
    ConstraintWeights_2:float=0.264
    ConstraintWeights_3:float=0.0071
    ConstraintWeights_4:float=0.0047
    ConstraintWeights_5:float=0.0021
   ) -> (Pose) {
    boneToModify {boneName=dyn_bow_left_root}
    constraintSetup [
     {targetBone {boneName=clavicle_l}}
     {targetBone {boneName=spine_05}}
     {targetBone {boneName=dyn_bow_left_root}}
     {targetBone {boneName=neck_01}}
     {targetBone {boneName=spine_03}}
     {targetBone {boneName=neck_02}}
    ]
    constraintWeights [1,1,1,1,1,1]
   }
   n8 = "Constraint - dyn_bow_right_root (PPPPPP)"!(
    ComponentPose=@n7
    Alpha=1
    ConstraintWeights_0=0.506
    ConstraintWeights_1:float=0.4098
    ConstraintWeights_2:float=0.066
    ConstraintWeights_3:float=0.0094
    ConstraintWeights_4:float=0.0064
    ConstraintWeights_5:float=0.0025
   ) -> (Pose) {
    boneToModify {boneName=dyn_bow_right_root}
    constraintSetup [
     {targetBone {boneName=clavicle_r}}
     {targetBone {boneName=spine_05}}
     {targetBone {boneName=dyn_bow_right_root}}
     {targetBone {boneName=neck_01}}
     {targetBone {boneName=spine_03}}
     {targetBone {boneName=neck_02}}
    ]
    constraintWeights [1,1,1,1,1,1]
   }
   n9 = Anim Dynamics - Bone: dyn_bowLoop_right_root_fw!(ComponentPose=@n8, Alpha=1, GravityOverride:Vector="0,0,-980") -> (Pose) {
    angularDampingOverride=1
    angularSpringConstant=3000
    bAngularSpring=true
    bOverrideAngularDamping=true
    bUseGravityOverride=true
    gravityScale=0.35
    simulationSpace=RootRelative
    boundBone {boneName=dyn_bowLoop_right_root_fw}
    componentAppliedLinearAccClamp=<3000 3000 3000>
    componentLinearAccScale=<0.5 0.5 0.5>
    componentLinearVelScale=<0.5 0.5 0.5>
    gravityOverride {z=-980}
    physicsBodyDefinitions [{
      boundBone {boneName=dyn_bowLoop_right_root_fw}
      boxExtents=<20 20 20>
      constraintSetup {angularLimitsMax=<5 0 15>, angularLimitsMin=<-5 -15 -15>, angularTarget=<1 0 0>}
      localJointOffset=<-4 0 0>
     }]
    relativeSpaceBone {boneName=spine_05}
    simSpaceSettings {maxAngularAcceleration=15, maxAngularVelocity=7, simSpaceAngularAlpha=0.7}
   }
   n10 = Anim Dynamics - Bone: dyn_bowLoop_right_root_back!(ComponentPose=@n9, Alpha=1, GravityOverride:Vector="0,0,-980") -> (Pose) {
    angularDampingOverride=1
    angularSpringConstant=3000
    bAngularSpring=true
    bOverrideAngularDamping=true
    bUseGravityOverride=true
    gravityScale=0.35
    simulationSpace=RootRelative
    boundBone {boneName=dyn_bowLoop_right_root_back}
    componentAppliedLinearAccClamp=<3000 3000 3000>
    componentLinearAccScale=<0.5 0.5 0.5>
    componentLinearVelScale=<0.5 0.5 0.5>
    gravityOverride {z=-980}
    physicsBodyDefinitions [{
      boundBone {boneName=dyn_bowLoop_right_root_back}
      boxExtents=<20 20 20>
      constraintSetup {angularLimitsMax=<5 5 9>, angularLimitsMin=<-5 -15 -15>, angularTarget=<1 0 0>}
      localJointOffset=<4 0 0>
     }]
    relativeSpaceBone {boneName=spine_05}
    simSpaceSettings {maxAngularAcceleration=15, maxAngularVelocity=7, simSpaceAngularAlpha=0.7}
   }
   n11 = Anim Dynamics - Bone: dyn_bowString_right_root!(ComponentPose=@n10, Alpha=1, GravityOverride:Vector="0,0,-980") -> (Pose) {
    angularDampingOverride=1
    angularSpringConstant=3000
    bAngularSpring=true
    bOverrideAngularDamping=true
    bUseGravityOverride=true
    gravityScale=0.35
    simulationSpace=RootRelative
    boundBone {boneName=dyn_bowString_right_root}
    componentAppliedLinearAccClamp=<3000 3000 3000>
    componentLinearAccScale=<0.5 0.5 0.5>
    componentLinearVelScale=<0.5 0.5 0.5>
    gravityOverride {z=-980}
    physicsBodyDefinitions [{
      sphereCollisionRadius=2
      boundBone {boneName=dyn_bowString_right_root}
      boxExtents=<20 20 20>
      constraintSetup {angularLimitsMax=<5 5 0>, angularLimitsMin=<-5 -15 -15>, angularTarget=<1 0 0>}
      localJointOffset=<4 0 0>
     }]
    relativeSpaceBone {boneName=spine_05}
    simSpaceSettings {maxAngularAcceleration=15, maxAngularVelocity=7, simSpaceAngularAlpha=0.7}
   }
   n12 = Anim Dynamics - Bone: dyn_bowLoop_left_root_fw!(ComponentPose=@n11, Alpha=1, GravityOverride:Vector="0,0,-980") -> (Pose) {
    angularDampingOverride=1
    angularSpringConstant=3000
    bAngularSpring=true
    bOverrideAngularDamping=true
    bUseGravityOverride=true
    gravityScale=0.35
    simulationSpace=RootRelative
    boundBone {boneName=dyn_bowLoop_left_root_fw}
    componentAppliedLinearAccClamp=<3000 3000 3000>
    componentLinearAccScale=<0.5 0.5 0.5>
    componentLinearVelScale=<0.5 0.5 0.5>
    gravityOverride {z=-980}
    physicsBodyDefinitions [{
      boundBone {boneName=dyn_bowLoop_left_root_fw}
      boxExtents=<20 20 20>
      constraintSetup {angularLimitsMax=<0 0 15>, angularLimitsMin=<0 -15 0>, angularTarget=<1 0 0>}
      localJointOffset=<-4 0 0>
     }]
    relativeSpaceBone {boneName=spine_05}
    simSpaceSettings {maxAngularAcceleration=15, maxAngularVelocity=7, simSpaceAngularAlpha=0.7}
   }
   n13 = Anim Dynamics - Bone: dyn_bowLoop_left_root_back!(ComponentPose=@n12, Alpha=1, GravityOverride:Vector="0,0,-980") -> (Pose) {
    angularDampingOverride=1
    angularSpringConstant=3000
    bAngularSpring=true
    bOverrideAngularDamping=true
    bUseGravityOverride=true
    gravityScale=0.35
    simulationSpace=RootRelative
    boundBone {boneName=dyn_bowLoop_left_root_back}
    componentAppliedLinearAccClamp=<3000 3000 3000>
    componentLinearAccScale=<0.5 0.5 0.5>
    componentLinearVelScale=<0.5 0.5 0.5>
    gravityOverride {z=-980}
    physicsBodyDefinitions [{
      boundBone {boneName=dyn_bowLoop_left_root_back}
      boxExtents=<20 20 20>
      constraintSetup {angularLimitsMax=<0 15 15>, angularLimitsMin=<0 -15 -15>, angularTarget=<1 0 0>}
      localJointOffset=<4 0 0>
     }]
    relativeSpaceBone {boneName=spine_05}
    simSpaceSettings {maxAngularAcceleration=15, maxAngularVelocity=7, simSpaceAngularAlpha=0.7}
   }
   n14 = Anim Dynamics - Bone: dyn_bowString_left_root!(ComponentPose=@n13, Alpha=1, GravityOverride:Vector="0,0,-980") -> (Pose) {
    angularDampingOverride=1
    angularSpringConstant=3000
    bAngularSpring=true
    bOverrideAngularDamping=true
    bUseGravityOverride=true
    gravityScale=0.35
    simulationSpace=RootRelative
    boundBone {boneName=dyn_bowString_left_root}
    componentAppliedLinearAccClamp=<3000 3000 3000>
    componentLinearAccScale=<0.5 0.5 0.5>
    componentLinearVelScale=<0.5 0.5 0.5>
    gravityOverride {z=-980}
    physicsBodyDefinitions [{
      sphereCollisionRadius=2
      boundBone {boneName=dyn_bowString_left_root}
      boxExtents=<20 20 20>
      constraintSetup {angularLimitsMax=<5 -5 10>, angularLimitsMin=<-5 -25 5>, angularTarget=<1 0 0>}
      localJointOffset=<4 0 0>
     }]
    relativeSpaceBone {boneName=spine_05}
    simSpaceSettings {maxAngularAcceleration=15, maxAngularVelocity=7, simSpaceAngularAlpha=0.7}
   }
   n15 = RigidBody!(
    ComponentPose=@n14
    Alpha
    OverrideWorldGravity:Vector="0,0,0"
    SimSpaceSettings:Sim Space Settings Structure=(WorldAlpha=1,VelocityScaleZ=1,DampingAlpha=0,MaxLinearVelocity=1400,MaxAngularVelocity=20,MaxLinearAcceleration=8,MaxAngularAcceleration=10,ExternalLinearDragV=(X=3,Y=1,Z=1),ExternalLinearVelocity=(X=0,Y=0,Z=0),ExternalAngularVelocity=(X=0,Y=0,Z=0))
   ) -> (Pose) {
    bOverrideWorldGravity=true
    overridePhysicsAsset=&/F_MED_CherryKnot_Strings_Physics
    simulationSpace=BaseBoneSpace
    baseBoneRef {boneName=pelvis}
    simSpaceSettings {
     dampingAlpha=0
     maxAngularAcceleration=10
     maxAngularVelocity=20
     maxLinearAcceleration=8
     maxLinearVelocity=1400
     worldAlpha=1
     externalLinearDragV=<3 1 1>
    }
   }
   n16 = RigidBody(
    ComponentPose=@n15
    OverrideWorldGravity="0,0,0"
    SimSpaceSettings=(WorldAlpha=1,VelocityScaleZ=1,DampingAlpha=0,MaxLinearVelocity=1400,MaxAngularVelocity=20,MaxLinearAcceleration=8,MaxAngularAcceleration=10,ExternalLinearDragV=(X=3,Y=1,Z=1),ExternalLinearVelocity=(X=0,Y=0,Z=0),ExternalAngularVelocity=(X=0,Y=0,Z=0))
   ) {
    bOverrideWorldGravity=true
    overridePhysicsAsset=&/F_MED_CherryKnot_Strings_Skirt_Physics
    simulationSpace=BaseBoneSpace
    baseBoneRef {boneName=pelvis}
    simSpaceSettings {
     dampingAlpha=0
     maxAngularAcceleration=10
     maxAngularVelocity=20
     maxLinearAcceleration=8
     maxLinearVelocity=1400
     worldAlpha=1
     externalLinearDragV=<3 1 1>
    }
   }
   Output Pose(Result=Component To Local(ComponentPose=@n16))
  ]}
 EventGraph {nodes [
   Event BlueprintUpdateAnimation()
   > n1 = Sequence()
   > n2 = "Gravity Override State Function (Parameter Asset) (Nativized)"(
    CharacterPartAnimInstance=Self-Reference()
    GameTimeInSeconds=GetGameTimeInSeconds()
    Parameters=&/Fortnite_CherryKnot_Strings_Parameters
   )
   > Set Gravity Override(Gravity Override=@n2)
   > Set RBAN Settings Out(RBAN Settings Out=@n2.OutRigidBodyAnimNodeParameters)
   Set frontendAlpha(execute=@n1.then_1, frontendAlpha:double=SelectFloat(A=0.25, B=1, bPickA=Get bIsInFrontEnd())) -> (Output_Get:double)
  ]}
]
variables [
 frontendAlpha {type=number}
 Gravity Override {type=Vector}
 RBAN Settings Out {type=FortRigidBodyAnimNodeParameters}
]
